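# M_UI_PixieParts is a translucent UI material, filed with the PixieParts body cosmetic, that procedurally draws a gradient rounded shape without textures.
# A two-color gradient blends rotated linear and circular forms, with flip, desaturation and vignette controls.
# A sine-based highlight band recolors through an HSV remap with tint or replace.
# MF_UI_CoreUI_RoundedMask supplies the shape mask and SDF, with parameters for roundness, size and blurred opacity.
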Material /BRCosmetics/Characters/Player/Male/Medium/Bodies/M_MED_PixieParts_Lava/MaterialParameterSets/UI/M_UI_PixieParts
graph {
 expressions {
  MaterialExpressionAdd_1 {
   class=MaterialExpressionAdd
   inputs [
    A {from=MaterialExpressionNamedRerouteUsage_33}
    B {from=""}
   ]
   properties {constB=0.25}
  }
  MaterialExpressionAdd_2 {
   class=MaterialExpressionAdd
   inputs [
    A {from=MaterialExpressionNamedRerouteUsage_6}
    B {from=MaterialExpressionConvert_2}
   ]
  }
  MaterialExpressionComponentMask_0 {class=MaterialExpressionComponentMask, inputs [None {from=MaterialExpressionNamedRerouteUsage_47}], properties {r=true}}
  MaterialExpressionComponentMask_5 {class=MaterialExpressionComponentMask, inputs [None {from=MaterialExpressionNamedRerouteUsage_46}], properties {r=true}}
  MaterialExpressionConstant2Vector_0 {class=MaterialExpressionConstant2Vector, properties {r=0.5}}
  MaterialExpressionConstant_0 {class=MaterialExpressionConstant, properties {r=1}}
  MaterialExpressionConvert_2 {
   class=MaterialExpressionConvert
   inputs [None {from=MaterialExpressionRgbToHsv_2}]
   properties {
    nodeName=Break Float3
    convertInputs [{
      type=Vector3
      expressionInput {
       expression=&self:MaterialExpressionRgbToHsv_2
       inputName=None
       mask=0
       maskA=0
       maskB=0
       maskG=0
       maskR=0
       outputIndex=0
      }
     }]
    convertOutputs [{type=Scalar}, {type=Scalar}, {type=Scalar}]
   }
  }
  MaterialExpressionConvert_3 {
   class=MaterialExpressionConvert
   inputs [
    None {from=MaterialExpressionFmod_0}
    None {from=MaterialExpressionMultiply_6}
    None {from=MaterialExpressionPower_1}
   ]
   properties {
    nodeName=Make Vector3
    convertInputs [
     {
      type=Scalar
      expressionInput {
       expression=&self:MaterialExpressionFmod_0
       inputName=None
       mask=0
       maskA=0
       maskB=0
       maskG=0
       maskR=0
       outputIndex=0
      }
     }
     {
      type=Scalar
      expressionInput {
       expression=&self:MaterialExpressionMultiply_6
       inputName=None
       mask=0
       maskA=0
       maskB=0
       maskG=0
       maskR=0
       outputIndex=0
      }
     }
     {
      type=Scalar
      expressionInput {
       expression=&self:MaterialExpressionPower_1
       inputName=None
       mask=0
       maskA=0
       maskB=0
       maskG=0
       maskR=0
       outputIndex=0
      }
     }
    ]
    convertOutputs [{type=Vector3}]
   }
  }
  MaterialExpressionConvert_4 {
   class=MaterialExpressionConvert
   inputs [
    None {from=MaterialExpressionFmod_0}
    None {from=MaterialExpressionReroute_3}
    None {from=""}
   ]
   properties {
    nodeName=Make Vector3
    convertInputs [
     {
      type=Scalar
      expressionInput {
       expression=&self:MaterialExpressionFmod_0
       inputName=None
       mask=0
       maskA=0
       maskB=0
       maskG=0
       maskR=0
       outputIndex=0
      }
     }
     {
      type=Scalar
      expressionInput {
       expression=&self:MaterialExpressionReroute_3
       inputName=None
       mask=0
       maskA=0
       maskB=0
       maskG=0
       maskR=0
       outputIndex=0
      }
     }
     {type=Scalar, expressionInput {expression=null, inputName=None, mask=0, maskA=0, maskB=0, maskG=0, maskR=0, outputIndex=0}}
    ]
    convertOutputs [{type=Vector3}]
   }
  }
  MaterialExpressionDesaturation_0 {
   class=MaterialExpressionDesaturation
   inputs [
    None {from=MaterialExpressionHsvToRgb_1, output=""}
    Fraction {from=MaterialExpressionOneMinus_3, output=""}
   ]
   properties {}
  }
  MaterialExpressionDesaturation_3 {
   class=MaterialExpressionDesaturation
   inputs [
    None {from=MaterialExpressionLinearInterpolate_3, output=""}
    Fraction {from=MaterialExpressionNamedRerouteUsage_29, output=""}
   ]
   properties {}
  }
  MaterialExpressionDivide_1 {
   class=MaterialExpressionDivide
   inputs [
    A {from=MaterialExpressionMultiply_19}
    B {from=MaterialExpressionNamedRerouteUsage_53}
   ]
  }
  MaterialExpressionDivide_2 {
   class=MaterialExpressionDivide
   inputs [
    A {from=""}
    B {from=MaterialExpressionNamedRerouteUsage_25}
   ]
  }
  MaterialExpressionFmod_0 {
   class=MaterialExpressionFmod
   inputs [
    A {from=MaterialExpressionAdd_2, output=""}
    B {from=MaterialExpressionConstant_0, output=""}
   ]
   properties {}
  }
  MaterialExpressionHsvToRgb_0 {class=MaterialExpressionHsvToRgb, inputs [None {from=MaterialExpressionConvert_3, output=""}], properties {}}
  MaterialExpressionHsvToRgb_1 {class=MaterialExpressionHsvToRgb, inputs [None {from=MaterialExpressionConvert_4, output=""}], properties {}}
  MaterialExpressionLength_1 {class=MaterialExpressionLength, inputs [None {from=MaterialExpressionSubtract_2, output=""}], properties {}}
  MaterialExpressionLinearInterpolate_0 {
   class=MaterialExpressionLinearInterpolate
   inputs [
    A {from=MaterialExpressionMultiply_24}
    B {from=MaterialExpressionNamedRerouteUsage_21}
    Alpha {from=MaterialExpressionMultiply_5}
   ]
  }
  MaterialExpressionLinearInterpolate_1 {
   class=MaterialExpressionLinearInterpolate
   inputs [
    A {from=MaterialExpressionNamedRerouteUsage_9}
    B {from=""}
    Alpha {from=MaterialExpressionNamedRerouteUsage_2}
   ]
   properties {constA=1}
  }
  MaterialExpressionLinearInterpolate_10 {
   class=MaterialExpressionLinearInterpolate
   inputs [
    A {from=MaterialExpressionMultiply_8}
    B {from=MaterialExpressionSaturate_0}
    Alpha {from=MaterialExpressionSaturate_1}
   ]
  }
  MaterialExpressionLinearInterpolate_14 {
   class=MaterialExpressionLinearInterpolate
   inputs [
    A {from=MaterialExpressionNamedRerouteUsage_52}
    B {from=MaterialExpressionSaturate_6}
    Alpha {from=MaterialExpressionNamedRerouteUsage_45}
   ]
  }
  MaterialExpressionLinearInterpolate_15 {
   class=MaterialExpressionLinearInterpolate
   inputs [
    A {from=MaterialExpressionMultiply_16}
    B {from=MaterialExpressionPower_8}
    Alpha {from=MaterialExpressionNamedRerouteUsage_26}
   ]
  }
  MaterialExpressionLinearInterpolate_16 {
   class=MaterialExpressionLinearInterpolate
   inputs [
    A {from=MaterialExpressionSaturate_9}
    B {from=MaterialExpressionOneMinus_7}
    Alpha {from=MaterialExpressionNamedRerouteUsage_28}
   ]
  }
  MaterialExpressionLinearInterpolate_18 {
   class=MaterialExpressionLinearInterpolate
   inputs [
    A {from=""}
    B {from=MaterialExpressionNamedRerouteUsage_57}
    Alpha {from=MaterialExpressionNamedRerouteUsage_32}
   ]
   properties {constA=1}
  }
  MaterialExpressionLinearInterpolate_3 {
   class=MaterialExpressionLinearInterpolate
   inputs [
    A {from=MaterialExpressionNamedRerouteUsage_31}
    B {from=MaterialExpressionNamedRerouteUsage_30}
    Alpha {from=MaterialExpressionNamedRerouteUsage_48}
   ]
  }
  MaterialExpressionLinearInterpolate_4 {
   class=MaterialExpressionLinearInterpolate
   inputs [
    A {from=MaterialExpressionHsvToRgb_0}
    B {from=MaterialExpressionDesaturation_0}
    Alpha {from=MaterialExpressionNamedRerouteUsage_1}
   ]
  }
  MaterialExpressionLinearInterpolate_5 {
   class=MaterialExpressionLinearInterpolate
   inputs [
    A {from=""}
    B {from=MaterialExpressionNamedRerouteUsage_12}
    Alpha {from=MaterialExpressionNamedRerouteUsage_11}
   ]
   properties {constA=1}
  }
  MaterialExpressionMaterialFunctionCall_2 {
   class=MaterialExpressionMaterialFunctionCall
   inputs [
    UV {from=""}
    Size {from=MaterialExpressionNamedRerouteUsage_42}
    Corner Radius {from=MaterialExpressionNamedRerouteUsage_43}
    Radius in Pixels {from=""}
    Skew Angle {from=""}
    Outline Thickness {from=""}
    Corners Modulated by Size {from=""}
    Pixel Size {from=""}
   ]
   properties {materialFunction=&MF_UI_CoreUI_RoundedMask}
  }
  MaterialExpressionMaterialFunctionCall_4 {
   class=MaterialExpressionMaterialFunctionCall
   inputs [
    UVs {from=MaterialExpressionTextureCoordinate_6}
    Rotation Center {from=""}
    "Rotation Angle (0-1)" {from=MaterialExpressionNamedRerouteUsage_27}
   ]
   properties {materialFunction=&CustomRotator}
  }
  MaterialExpressionMax_1 {
   class=MaterialExpressionMax
   inputs [
    A {from=MaterialExpressionVectorParameter_5, output=Saturation}
    B {from=""}
   ]
   properties {constB=0}
  }
  MaterialExpressionMax_4 {
   class=MaterialExpressionMax
   inputs [
    A {from=MaterialExpressionReroute_2}
    B {from=""}
   ]
   properties {constB=0}
  }
  MaterialExpressionMultiply_1 {
   class=MaterialExpressionMultiply
   inputs [
    A {from=MaterialExpressionSubtract_0}
    B {from=MaterialExpressionNamedRerouteUsage_34}
   ]
  }
  MaterialExpressionMultiply_10 {
   class=MaterialExpressionMultiply
   inputs [
    A {from=MaterialExpressionNamedRerouteUsage_51}
    B {from=MaterialExpressionNamedRerouteUsage_44}
   ]
  }
  MaterialExpressionMultiply_16 {
   class=MaterialExpressionMultiply
   inputs [
    A {from=MaterialExpressionComponentMask_5}
    B {from=MaterialExpressionNamedRerouteUsage_25}
   ]
  }
  MaterialExpressionMultiply_18 {
   class=MaterialExpressionMultiply
   inputs [
    A {from=MaterialExpressionLinearInterpolate_0}
    B {from=MaterialExpressionOneMinus_8}
   ]
  }
  MaterialExpressionMultiply_19 {
   class=MaterialExpressionMultiply
   inputs [
    A {from=MaterialExpressionSubtract_3}
    B {from=""}
   ]
   properties {constB=-1}
  }
  MaterialExpressionMultiply_21 {
   class=MaterialExpressionMultiply
   inputs [
    A {from=MaterialExpressionLength_1}
    B {from=""}
   ]
   properties {constB=2}
  }
  MaterialExpressionMultiply_22 {
   class=MaterialExpressionMultiply
   inputs [
    A {from=MaterialExpressionPower_5}
    B {from=MaterialExpressionNamedRerouteUsage_36}
   ]
  }
  MaterialExpressionMultiply_24 {
   class=MaterialExpressionMultiply
   inputs [
    A {from=MaterialExpressionNamedRerouteUsage_50}
    B {from=MaterialExpressionLinearInterpolate_18}
   ]
  }
  MaterialExpressionMultiply_5 {
   class=MaterialExpressionMultiply
   inputs [
    A {from=MaterialExpressionNamedRerouteUsage_0}
    B {from=MaterialExpressionNamedRerouteUsage_32}
   ]
  }
  MaterialExpressionMultiply_6 {
   class=MaterialExpressionMultiply
   inputs [
    A {from=MaterialExpressionConvert_2}
    B {from=MaterialExpressionNamedRerouteUsage_7}
   ]
  }
  MaterialExpressionMultiply_7 {
   class=MaterialExpressionMultiply
   inputs [
    A {from=MaterialExpressionConvert_2}
    B {from=MaterialExpressionNamedRerouteUsage_8}
   ]
  }
  MaterialExpressionMultiply_8 {
   class=MaterialExpressionMultiply
   inputs [
    A {from=MaterialExpressionPower_4}
    B {from=MaterialExpressionNamedRerouteUsage_17}
   ]
  }
  MaterialExpressionNamedRerouteDeclaration_0 {
   class=MaterialExpressionNamedRerouteDeclaration
   inputs [None {from=MaterialExpressionPower_0}]
   properties {name="Highlight Mask (S)", nodeColor=<rgba 1 0 0.5625 1>}
  }
  MaterialExpressionNamedRerouteDeclaration_1 {
   class=MaterialExpressionNamedRerouteDeclaration
   inputs [None {from=MaterialExpressionMax_1}]
   properties {name="Highlight Sat (S)", nodeColor=<rgba 0.7652 0 1 1>}
  }
  MaterialExpressionNamedRerouteDeclaration_14 {
   class=MaterialExpressionNamedRerouteDeclaration
   inputs [None {from=MaterialExpressionSaturate_2}]
   properties {name="Highlight Sat - Preserve Luminance (S)", nodeColor=<rgba 1 0 0.1875 1>}
  }
  MaterialExpressionNamedRerouteDeclaration_15 {
   class=MaterialExpressionNamedRerouteDeclaration
   inputs [None {from=MaterialExpressionSaturate_3}]
   properties {name="Higlight Exp - Value <-> Color (S)", nodeColor=<rgba 0 0.6957 1 1>}
  }
  MaterialExpressionNamedRerouteDeclaration_18 {
   class=MaterialExpressionNamedRerouteDeclaration
   inputs [None {from=MaterialExpressionDesaturation_3}]
   properties {name="Gradient Color (V3)", nodeColor=<rgba 0.7578 0 1 1>}
  }
  MaterialExpressionNamedRerouteDeclaration_19 {
   class=MaterialExpressionNamedRerouteDeclaration
   inputs [None {from=MaterialExpressionLinearInterpolate_14}]
   properties {name="Opacity (S)", nodeColor=<rgba 0 0.1562 1 1>}
  }
  MaterialExpressionNamedRerouteDeclaration_2 {
   class=MaterialExpressionNamedRerouteDeclaration
   inputs [None {from=MaterialExpressionReroute_1}]
   properties {name="Highlight Exp (S)", nodeColor=<rgba 1 0.2109 0 1>}
  }
  MaterialExpressionNamedRerouteDeclaration_21 {
   class=MaterialExpressionNamedRerouteDeclaration
   inputs [None {from=MaterialExpressionMaterialFunctionCall_2, output=SDF}]
   properties {name="Shape SDF (S)", nodeColor=<rgba 0.3828 0 1 1>}
  }
  MaterialExpressionNamedRerouteDeclaration_22 {
   class=MaterialExpressionNamedRerouteDeclaration
   inputs [None {from=MaterialExpressionScalarParameter_12}]
   properties {name="Gradient - Secondary Color Mix (S)", nodeColor=<rgba 0.7812 1 0 1>}
  }
  MaterialExpressionNamedRerouteDeclaration_23 {
   class=MaterialExpressionNamedRerouteDeclaration
   inputs [None {from=MaterialExpressionSaturate_10}]
   properties {name="Gradient - Linear <-> Circular (S)", nodeColor=<rgba 0.8984 0 1 1>}
  }
  MaterialExpressionNamedRerouteDeclaration_24 {
   class=MaterialExpressionNamedRerouteDeclaration
   inputs [None {from=MaterialExpressionScalarParameter_16}]
   properties {name="Gradient - Linear Angle (S)", nodeColor=<rgba 0 1 0.5781 1>}
  }
  MaterialExpressionNamedRerouteDeclaration_25 {
   class=MaterialExpressionNamedRerouteDeclaration
   inputs [None {from=MaterialExpressionSaturate_11}]
   properties {name="Gradient - Flip Colors (S)", nodeColor=<rgba 1 0.2578 0 1>}
  }
  MaterialExpressionNamedRerouteDeclaration_26 {
   class=MaterialExpressionNamedRerouteDeclaration
   inputs [None {from=MaterialExpressionScalarParameter_23}]
   properties {name="Gradient - Desaturate (S)", nodeColor=<rgba 0 0.0625 1 1>}
  }
  MaterialExpressionNamedRerouteDeclaration_27 {
   class=MaterialExpressionNamedRerouteDeclaration
   inputs [None {from=MaterialExpressionVectorParameter_1, output=RGB}]
   properties {name="Gradient - Primary Color (V3)", nodeColor=<rgba 0.3828 1 0 1>}
  }
  MaterialExpressionNamedRerouteDeclaration_28 {
   class=MaterialExpressionNamedRerouteDeclaration
   inputs [None {from=MaterialExpressionVectorParameter_0, output=RGB}]
   properties {name="Gradient - Secondary Color (V3)", nodeColor=<rgba 1 0 0.7031 1>}
  }
  MaterialExpressionNamedRerouteDeclaration_29 {
   class=MaterialExpressionNamedRerouteDeclaration
   inputs [None {from=MaterialExpressionScalarParameter_3}]
   properties {name="Highlight - Mix Amout (S)", nodeColor=<rgba 0 1 0.9766 1>}
  }
  MaterialExpressionNamedRerouteDeclaration_3 {
   class=MaterialExpressionNamedRerouteDeclaration
   inputs [None {from=MaterialExpressionMax_4}]
   properties {name="Highlight Val (S)", nodeColor=<rgba 0 0.1094 1 1>}
  }
  MaterialExpressionNamedRerouteDeclaration_31 {
   class=MaterialExpressionNamedRerouteDeclaration
   inputs [None {from=MaterialExpressionScalarParameter_0}]
   properties {name="Highlight - Offset (S)", nodeColor=<rgba 1 0.6563 0 1>}
  }
  MaterialExpressionNamedRerouteDeclaration_32 {
   class=MaterialExpressionNamedRerouteDeclaration
   inputs [None {from=MaterialExpressionScalarParameter_1}]
   properties {name="Highlight - Frequency (S)", nodeColor=<rgba 0.3359 0 1 1>}
  }
  MaterialExpressionNamedRerouteDeclaration_33 {
   class=MaterialExpressionNamedRerouteDeclaration
   inputs [None {from=MaterialExpressionScalarParameter_2}]
   properties {name="Highlight - Tightness (S)", nodeColor=<rgba 0 1 0.0156 1>}
  }
  MaterialExpressionNamedRerouteDeclaration_34 {
   class=MaterialExpressionNamedRerouteDeclaration
   inputs [None {from=MaterialExpressionScalarParameter_20}]
   properties {name="Gradient - Vignette Intensity (S)", nodeColor=<rgba 1 0 0.3047 1>}
  }
  MaterialExpressionNamedRerouteDeclaration_35 {
   class=MaterialExpressionNamedRerouteDeclaration
   inputs [None {from=MaterialExpressionScalarParameter_6}]
   properties {name="Shape - Size (S)", nodeColor=<rgba 0 0.625 1 1>}
  }
  MaterialExpressionNamedRerouteDeclaration_36 {
   class=MaterialExpressionNamedRerouteDeclaration
   inputs [None {from=MaterialExpressionSaturate_15}]
   properties {name="Shape  - Corner Radius (S)", nodeColor=<rgba 0.9453 1 0 1>}
  }
  MaterialExpressionNamedRerouteDeclaration_37 {
   class=MaterialExpressionNamedRerouteDeclaration
   inputs [None {from=MaterialExpressionScalarParameter_8}]
   properties {name="Shape - Blur Inv Size (S)", nodeColor=<rgba 0.7344 0 1 1>}
  }
  MaterialExpressionNamedRerouteDeclaration_39 {
   class=MaterialExpressionNamedRerouteDeclaration
   inputs [None {from=MaterialExpressionSaturate_12}]
   properties {name="Shape - Blur Mix (S)", nodeColor=<rgba 0 1 0.4141 1>}
  }
  MaterialExpressionNamedRerouteDeclaration_4 {
   class=MaterialExpressionNamedRerouteDeclaration
   inputs [None {from=MaterialExpressionSaturate_4}]
   properties {name="Vignette (S)", nodeColor=<rgba 0.0078 0 1 1>}
  }
  MaterialExpressionNamedRerouteDeclaration_40 {
   class=MaterialExpressionNamedRerouteDeclaration
   inputs [None {from=MaterialExpressionMaterialFunctionCall_4, output=Rotated Values}]
   properties {name="Rotated UV (V2)", nodeColor=<rgba 0.1719 1 0 1>}
  }
  MaterialExpressionNamedRerouteDeclaration_41 {
   class=MaterialExpressionNamedRerouteDeclaration
   inputs [None {from=MaterialExpressionLinearInterpolate_16}]
   properties {name="Gradient (S)", nodeColor=<rgba 1 0 0.4922 1>}
  }
  MaterialExpressionNamedRerouteDeclaration_43 {
   class=MaterialExpressionNamedRerouteDeclaration
   inputs [None {from=MaterialExpressionMaterialFunctionCall_2, output=Shape Mask}]
   properties {name="Shape Mask (S)", nodeColor=<rgba 1 0.8672 0 1>}
  }
  MaterialExpressionNamedRerouteDeclaration_44 {
   class=MaterialExpressionNamedRerouteDeclaration
   inputs [None {from=MaterialExpressionScalarParameter_19}]
   properties {name="Highlight - Base Gradient Mult (S)", nodeColor=<rgba 0 1 0.0156 1>}
  }
  MaterialExpressionNamedRerouteDeclaration_5 {
   class=MaterialExpressionNamedRerouteDeclaration
   inputs [None {from=MaterialExpressionReroute_0}]
   properties {name="Highlight Hue (S)", nodeColor=<rgba 0.4297 1 0 1>}
  }
  MaterialExpressionNamedRerouteDeclaration_6 {
   class=MaterialExpressionNamedRerouteDeclaration
   inputs [None {from=MaterialExpressionVectorParameter_4, output=RGB}]
   properties {name="Highlight Tint (V3)", nodeColor=<rgba 0.0312 1 0 1>}
  }
  MaterialExpressionNamedRerouteDeclaration_7 {
   class=MaterialExpressionNamedRerouteDeclaration
   inputs [None {from=MaterialExpressionVectorParameter_4, output="Tint <-> Replace"}]
   properties {name="Highlight Tint <-> Replace (S)", nodeColor=<rgba 1 0 0.3516 1>}
  }
  MaterialExpressionNamedRerouteDeclaration_8 {
   class=MaterialExpressionNamedRerouteDeclaration
   inputs [None {from=MaterialExpressionLinearInterpolate_10}]
   properties {name="Highlight (V3)", nodeColor=<rgba 1 0.5217 0 1>}
  }
  MaterialExpressionNamedRerouteUsage_0 {class=MaterialExpressionNamedRerouteUsage}
  MaterialExpressionNamedRerouteUsage_1 {class=MaterialExpressionNamedRerouteUsage}
  MaterialExpressionNamedRerouteUsage_10 {class=MaterialExpressionNamedRerouteUsage}
  MaterialExpressionNamedRerouteUsage_11 {class=MaterialExpressionNamedRerouteUsage}
  MaterialExpressionNamedRerouteUsage_12 {class=MaterialExpressionNamedRerouteUsage}
  MaterialExpressionNamedRerouteUsage_17 {class=MaterialExpressionNamedRerouteUsage}
  MaterialExpressionNamedRerouteUsage_18 {class=MaterialExpressionNamedRerouteUsage}
  MaterialExpressionNamedRerouteUsage_2 {class=MaterialExpressionNamedRerouteUsage}
  MaterialExpressionNamedRerouteUsage_20 {class=MaterialExpressionNamedRerouteUsage}
  MaterialExpressionNamedRerouteUsage_21 {class=MaterialExpressionNamedRerouteUsage}
  MaterialExpressionNamedRerouteUsage_22 {class=MaterialExpressionNamedRerouteUsage}
  MaterialExpressionNamedRerouteUsage_25 {class=MaterialExpressionNamedRerouteUsage}
  MaterialExpressionNamedRerouteUsage_26 {class=MaterialExpressionNamedRerouteUsage}
  MaterialExpressionNamedRerouteUsage_27 {class=MaterialExpressionNamedRerouteUsage}
  MaterialExpressionNamedRerouteUsage_28 {class=MaterialExpressionNamedRerouteUsage}
  MaterialExpressionNamedRerouteUsage_29 {class=MaterialExpressionNamedRerouteUsage}
  MaterialExpressionNamedRerouteUsage_3 {class=MaterialExpressionNamedRerouteUsage}
  MaterialExpressionNamedRerouteUsage_30 {class=MaterialExpressionNamedRerouteUsage}
  MaterialExpressionNamedRerouteUsage_31 {class=MaterialExpressionNamedRerouteUsage}
  MaterialExpressionNamedRerouteUsage_32 {class=MaterialExpressionNamedRerouteUsage}
  MaterialExpressionNamedRerouteUsage_33 {class=MaterialExpressionNamedRerouteUsage}
  MaterialExpressionNamedRerouteUsage_34 {class=MaterialExpressionNamedRerouteUsage}
  MaterialExpressionNamedRerouteUsage_35 {class=MaterialExpressionNamedRerouteUsage}
  MaterialExpressionNamedRerouteUsage_36 {class=MaterialExpressionNamedRerouteUsage}
  MaterialExpressionNamedRerouteUsage_42 {class=MaterialExpressionNamedRerouteUsage}
  MaterialExpressionNamedRerouteUsage_43 {class=MaterialExpressionNamedRerouteUsage}
  MaterialExpressionNamedRerouteUsage_44 {class=MaterialExpressionNamedRerouteUsage}
  MaterialExpressionNamedRerouteUsage_45 {class=MaterialExpressionNamedRerouteUsage}
  MaterialExpressionNamedRerouteUsage_46 {class=MaterialExpressionNamedRerouteUsage}
  MaterialExpressionNamedRerouteUsage_47 {class=MaterialExpressionNamedRerouteUsage}
  MaterialExpressionNamedRerouteUsage_48 {class=MaterialExpressionNamedRerouteUsage}
  MaterialExpressionNamedRerouteUsage_50 {class=MaterialExpressionNamedRerouteUsage}
  MaterialExpressionNamedRerouteUsage_51 {class=MaterialExpressionNamedRerouteUsage}
  MaterialExpressionNamedRerouteUsage_52 {class=MaterialExpressionNamedRerouteUsage}
  MaterialExpressionNamedRerouteUsage_53 {class=MaterialExpressionNamedRerouteUsage}
  MaterialExpressionNamedRerouteUsage_57 {class=MaterialExpressionNamedRerouteUsage}
  MaterialExpressionNamedRerouteUsage_6 {class=MaterialExpressionNamedRerouteUsage}
  MaterialExpressionNamedRerouteUsage_7 {class=MaterialExpressionNamedRerouteUsage}
  MaterialExpressionNamedRerouteUsage_8 {class=MaterialExpressionNamedRerouteUsage}
  MaterialExpressionNamedRerouteUsage_9 {class=MaterialExpressionNamedRerouteUsage}
  MaterialExpressionOneMinus_0 {class=MaterialExpressionOneMinus, inputs [None {from=MaterialExpressionSphereMask_0}]}
  MaterialExpressionOneMinus_3 {class=MaterialExpressionOneMinus, inputs [None {from=MaterialExpressionNamedRerouteUsage_10}]}
  MaterialExpressionOneMinus_5 {class=MaterialExpressionOneMinus, inputs [None {from=MaterialExpressionMultiply_10}]}
  MaterialExpressionOneMinus_7 {class=MaterialExpressionOneMinus, inputs [None {from=MaterialExpressionSaturate_9}]}
  MaterialExpressionOneMinus_8 {class=MaterialExpressionOneMinus, inputs [None {from=MaterialExpressionNamedRerouteUsage_3}]}
  MaterialExpressionPower_0 {
   class=MaterialExpressionPower
   inputs [
    Base {from=MaterialExpressionSine_0}
    Exp {from=MaterialExpressionNamedRerouteUsage_35}
   ]
   properties {constExponent=1}
  }
  MaterialExpressionPower_1 {
   class=MaterialExpressionPower
   inputs [
    Base {from=MaterialExpressionMultiply_7}
    Exp {from=MaterialExpressionLinearInterpolate_1}
   ]
  }
  MaterialExpressionPower_4 {
   class=MaterialExpressionPower
   inputs [
    Base {from=MaterialExpressionLinearInterpolate_4}
    Exp {from=MaterialExpressionLinearInterpolate_5}
   ]
  }
  MaterialExpressionPower_5 {
   class=MaterialExpressionPower
   inputs [
    Base {from=MaterialExpressionOneMinus_0}
    Exp {from=""}
   ]
   properties {constExponent=3}
  }
  MaterialExpressionPower_8 {
   class=MaterialExpressionPower
   inputs [
    Base {from=MaterialExpressionDivide_1}
    Exp {from=MaterialExpressionDivide_2}
   ]
  }
  MaterialExpressionReroute_0 {class=MaterialExpressionReroute, inputs [None {from=MaterialExpressionVectorParameter_5, output=Hue}]}
  MaterialExpressionReroute_1 {class=MaterialExpressionReroute, inputs [None {from=MaterialExpressionVectorParameter_5, output=Exponent}]}
  MaterialExpressionReroute_2 {class=MaterialExpressionReroute, inputs [None {from=MaterialExpressionVectorParameter_5, output=Value}]}
  MaterialExpressionReroute_3 {class=MaterialExpressionReroute, inputs [None {from=MaterialExpressionConvert_2}]}
  MaterialExpressionRgbToHsv_2 {class=MaterialExpressionRgbToHsv, inputs [None {from=MaterialExpressionNamedRerouteUsage_20, output=""}], properties {}}
  MaterialExpressionSaturate_0 {class=MaterialExpressionSaturate, inputs [None {from=MaterialExpressionNamedRerouteUsage_17}]}
  MaterialExpressionSaturate_1 {class=MaterialExpressionSaturate, inputs [None {from=MaterialExpressionNamedRerouteUsage_18}]}
  MaterialExpressionSaturate_10 {class=MaterialExpressionSaturate, inputs [None {from=MaterialExpressionScalarParameter_13}]}
  MaterialExpressionSaturate_11 {class=MaterialExpressionSaturate, inputs [None {from=MaterialExpressionScalarParameter_15}]}
  MaterialExpressionSaturate_12 {class=MaterialExpressionSaturate, inputs [None {from=MaterialExpressionScalarParameter_10}]}
  MaterialExpressionSaturate_15 {class=MaterialExpressionSaturate, inputs [None {from=MaterialExpressionScalarParameter_25}]}
  MaterialExpressionSaturate_2 {class=MaterialExpressionSaturate, inputs [None {from=MaterialExpressionScalarParameter_4}]}
  MaterialExpressionSaturate_3 {class=MaterialExpressionSaturate, inputs [None {from=MaterialExpressionScalarParameter_5}]}
  MaterialExpressionSaturate_4 {class=MaterialExpressionSaturate, inputs [None {from=MaterialExpressionMultiply_22}]}
  MaterialExpressionSaturate_6 {class=MaterialExpressionSaturate, inputs [None {from=MaterialExpressionOneMinus_5}]}
  MaterialExpressionSaturate_9 {class=MaterialExpressionSaturate, inputs [None {from=MaterialExpressionLinearInterpolate_15}]}
  MaterialExpressionScalarParameter_0 {class=MaterialExpressionScalarParameter, properties {group=Highlight, parameterName=Highlight - Offset, sortPriority=30}}
  MaterialExpressionScalarParameter_1 {class=MaterialExpressionScalarParameter, properties {defaultValue=1, group=Highlight, parameterName=Highlight - Frequency, sortPriority=40}}
  MaterialExpressionScalarParameter_10 {class=MaterialExpressionScalarParameter, properties {group=Shape, parameterName=Shape - Blur Mix, sliderMax=1, sortPriority=30}}
  MaterialExpressionScalarParameter_12 {
   class=MaterialExpressionScalarParameter
   properties {defaultValue=1, group=Gradient, parameterName=Gradient - Secondary Color Mix, sortPriority=50}
  }
  MaterialExpressionScalarParameter_13 {
   class=MaterialExpressionScalarParameter
   properties {group=Gradient, parameterName="Gradient - Linear <-> Circular", sliderMax=1, sortPriority=10}
  }
  MaterialExpressionScalarParameter_15 {class=MaterialExpressionScalarParameter, properties {group=Gradient, parameterName=Gradient - Flip Colors, sliderMax=1, sortPriority=60}}
  MaterialExpressionScalarParameter_16 {
   class=MaterialExpressionScalarParameter
   properties {defaultValue=0.35, group=Gradient, parameterName=Gradient - Linear Gradient Angle, sliderMax=1, sortPriority=20}
  }
  MaterialExpressionScalarParameter_19 {
   class=MaterialExpressionScalarParameter
   properties {defaultValue=1, group=Highlight, parameterName=Highlight - Base Gradient Mult, sliderMax=1, sortPriority=15}
  }
  MaterialExpressionScalarParameter_2 {class=MaterialExpressionScalarParameter, properties {defaultValue=5, group=Highlight, parameterName=Highlight - Tightness, sortPriority=20}}
  MaterialExpressionScalarParameter_20 {
   class=MaterialExpressionScalarParameter
   properties {defaultValue=2.75, group=Gradient, parameterName=Gradient - Vignette Intensity, sortPriority=80}
  }
  MaterialExpressionScalarParameter_23 {class=MaterialExpressionScalarParameter, properties {group=Gradient, parameterName=Gradient - Desaturate, sortPriority=70}}
  MaterialExpressionScalarParameter_25 {class=MaterialExpressionScalarParameter, properties {defaultValue=0.2, group=Shape, parameterName=Shape - Roundness, sortPriority=10}}
  MaterialExpressionScalarParameter_3 {class=MaterialExpressionScalarParameter, properties {group=Highlight, parameterName=Highlight - Mix Amount, sortPriority=10}}
  MaterialExpressionScalarParameter_4 {
   class=MaterialExpressionScalarParameter
   properties {group=Highlight, parameterName=Highlight - Saturation Preserves sRGB Luminance, sliderMax=1, sortPriority=60}
  }
  MaterialExpressionScalarParameter_5 {
   class=MaterialExpressionScalarParameter
   properties {group=Highlight, parameterName="Highlight - Value Exp <-> Color Exp", sliderMax=1, sortPriority=70}
  }
  MaterialExpressionScalarParameter_6 {class=MaterialExpressionScalarParameter, properties {defaultValue=1, group=Shape, parameterName=Shape - Size, sortPriority=20}}
  MaterialExpressionScalarParameter_8 {class=MaterialExpressionScalarParameter, properties {defaultValue=3, group=Shape, parameterName=Shape - Blur Inv Size, sortPriority=40}}
  MaterialExpressionSine_0 {class=MaterialExpressionSine, inputs [None {from=MaterialExpressionMultiply_1}]}
  MaterialExpressionSphereMask_0 {
   class=MaterialExpressionSphereMask
   inputs [
    A {from=MaterialExpressionTextureCoordinate_5, output=""}
    B {from=MaterialExpressionConstant2Vector_0, output=""}
    Radius {from="", output=""}
    Hardness {from="", output=""}
   ]
   properties {attenuationRadius=2, hardnessPercent=0}
  }
  MaterialExpressionSubtract_0 {
   class=MaterialExpressionSubtract
   inputs [
    A {from=MaterialExpressionComponentMask_0}
    B {from=MaterialExpressionAdd_1}
   ]
   properties {constB=0.25}
  }
  MaterialExpressionSubtract_2 {
   class=MaterialExpressionSubtract
   inputs [
    A {from=MaterialExpressionTextureCoordinate_1}
    B {from=""}
   ]
   properties {constB=0.5}
  }
  MaterialExpressionSubtract_3 {
   class=MaterialExpressionSubtract
   inputs [
    A {from=MaterialExpressionMultiply_21}
    B {from=MaterialExpressionNamedRerouteUsage_53}
   ]
  }
  MaterialExpressionTextureCoordinate_1 {class=MaterialExpressionTextureCoordinate, properties {coordinateIndex=4}}
  MaterialExpressionTextureCoordinate_5 {class=MaterialExpressionTextureCoordinate, properties {coordinateIndex=4}}
  MaterialExpressionTextureCoordinate_6 {class=MaterialExpressionTextureCoordinate}
  MaterialExpressionVectorParameter_0 {
   class=MaterialExpressionVectorParameter
   properties {group=Gradient, parameterName=Gradient - Secondary Color, sortPriority=40, defaultValue=<rgba 0.125 0.125 0.125 1>}
  }
  MaterialExpressionVectorParameter_1 {
   class=MaterialExpressionVectorParameter
   properties {group=Gradient, parameterName=Gradient - Primary Color, sortPriority=30, defaultValue=<rgba 0.12 0.12 0.12 1>}
  }
  MaterialExpressionVectorParameter_4 {
   class=MaterialExpressionVectorParameter
   properties {
    group=Highlight
    parameterName=Highlight - Tint
    sortPriority=80
    channelNames {a="Tint <-> Replace", b=Color B, g=Color G, r=Color R}
    defaultValue=<rgba 1 1 1 0>
   }
  }
  MaterialExpressionVectorParameter_5 {
   class=MaterialExpressionVectorParameter
   properties {
    group=Highlight
    parameterName=Highlight - Gradient Remap
    sortPriority=50
    channelNames {a=Exponent, b=Value, g=Saturation, r=Hue}
    defaultValue=<rgba 0 1 1 1>
   }
  }
 }
 properties {blendMode=BLEND_Translucent, materialDomain=MD_UI, usedShadingModels=Default Lit}
}
